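# Unlit generation material that produces tiling curl noise for encoding into a volume texture laid out as a flipbook grid of slices.
# EncodeVolumeCoordinates converts the XY frame layout into volume coordinates, which are scaled and fed as position into a tiling Vector Noise node using the Curl function.
# Negative emissive color is allowed, so signed vector values can be written without clamping.
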Material /Volumetrics/Content/VolumeTextures/Materials/Generation/M_Encode_Tiling_Curlnoise
graph {
 properties {bAllowNegativeEmissiveColor=true, pixelDepthOffsetMode=PDOM_Legacy, shadingModel=MSM_Unlit, usedShadingModels=Unlit}
 nodes [
  n1 = Constant2Vector() {g=8, r=16}
  VectorNoise(World Position=Multiply(A=MaterialFunctionCall!(UVs, XY Frames=@n1) {materialFunction=/Volumetrics/Content/VolumeTextures/MaterialFunctions/EncodeVolumeCoordinates}.Result) {constB=9}) {bTiling=true, desc="Perline Curl, Tile Size = 27", noiseFunction=VNF_CurlALU, tileSize=27}
 ]
}
