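# Unlit material for generating tiling 3D noise encoded across a 2D grid of frames, apparently for baking volume textures.
# EncodeVolumeCoordinates turns UVs and the XYFrames layout into volume positions, scaled by Tiling.
# A Custom HLSL loop sums Noise3D_Multiplexer gradient octaves weighted by Persistence, doubling position and repeat size together to keep tiling, with a turbulence option.
# A built-in tiling Noise node also samples that position; no output connection is shown.
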Material /Volumetrics/Content/VolumeTextures/Materials/Generation/M_Encode_Tiling_Noise
graph {
 properties {bAllowNegativeEmissiveColor=true, pixelDepthOffsetMode=PDOM_Legacy, shadingModel=MSM_Unlit, usedShadingModels=Unlit}
 nodes [
  MaterialExpressionScalarParameter_0 = ScalarParameter() {defaultValue=16, parameterName=Tiling}
  MaterialExpressionMultiply_8 = Multiply(
   A=MaterialFunctionCall!(UVs, XY Frames=VectorParameter() {parameterName=XYFrames, defaultValue=<rgba 16 16 0 0>}.RGB) {materialFunction=/Volumetrics/Content/VolumeTextures/MaterialFunctions/EncodeVolumeCoordinates}.Result
   B=@MaterialExpressionScalarParameter_0
  ) {constB=16}
  MaterialExpressionConstant_2 = Constant() {desc="3 = Gradient, 5 = Voronoi", r=3}
  MaterialExpressionConstant_3 = Constant()
  MaterialExpressionMin_0 = Min(A=ScalarParameter() {defaultValue=6, parameterName=Octaves}) {constB=10}
  MaterialExpressionScalarParameter_1 = ScalarParameter() {defaultValue=0.4, parameterName=Persistence}
  Custom!(
   Position=@MaterialExpressionMultiply_8
   Function=@MaterialExpressionConstant_2
   bTurbulence=@MaterialExpressionConstant_3
   Octaves=@MaterialExpressionMin_0
   Persistence=@MaterialExpressionScalarParameter_1
   RepeatSize=@MaterialExpressionScalarParameter_0
  ) {
   code="\r\nfloat Out = 0;\r\nfloat OutScale = 1;\r\n\r\n[loop]\r\nfor(int i = 0; i < Octaves; ++i)\r\n{\r\n\r\n\r\n\tif(bTurbulence)\r\n\t{\r\n\t\tOut += abs(Noise3D_Multiplexer(Function, Position, 4, 1, RepeatSize)) * OutScale;\r\n\t}\r\n\telse\r\n\t{\r\n\t\tOut += Noise3D_Multiplexer(Function, Position, 4, 1, RepeatSize) * OutScale;\r\n\t}\r\n\r\n\tPosition *= 2;\r\n\tRepeatSize *= 2;\r\n\tOutScale *= Persistence;\r\n\r\n}\r\n\r\nif(!bTurbulence)\r\n{\r\n\t// bring -1..1 to 0..1 range\r\n\tOut = Out * 0.5f + 0.5f;\r\n}\r\n\r\n// Out is in 0..1 range\r\nreturn lerp(0, 1, Out);"
   description=CustomNoise
   outputType=CMOT_Float1
   inputs [
    {
     inputName=Position
     input {expression=&self:MaterialExpressionMultiply_8}
    }
    {
     inputName=Function
     input {expression=&self:MaterialExpressionConstant_2}
    }
    {
     inputName=bTurbulence
     input {expression=&self:MaterialExpressionConstant_3}
    }
    {inputName=Octaves, input {expression=&self:MaterialExpressionMin_0}}
    {
     inputName=Persistence
     input {expression=&self:MaterialExpressionScalarParameter_1}
    }
    {
     inputName=RepeatSize
     input {expression=&self:MaterialExpressionScalarParameter_0}
    }
   ]
  }
  Noise!(World Position=@MaterialExpressionMultiply_8, FilterWidth) {bTiling=true, bTurbulence=false, noiseFunction=NOISEFUNCTION_GradientALU, repeatSize=16}
 ]
}
