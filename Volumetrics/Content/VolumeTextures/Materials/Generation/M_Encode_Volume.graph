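# Unlit material that generates an animated procedural volume laid out as a 16x16 grid of slices through EncodeVolumeCoordinates.
# Two time-panned, aspect-scaled gradient noise layers are weighted and added to an inverted, aspect-scaled radial falloff around point P with radius R.
# The sum is scaled by Temp, saturated, raised to TempPower and inverted.
# A curl-noise offset and a thresholded-coordinate branch feed nothing downstream, and no material output links are shown.
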Material /Volumetrics/Content/VolumeTextures/Materials/Generation/M_Encode_Volume
graph {
 properties {blendMode=BLEND_AlphaComposite, pixelDepthOffsetMode=PDOM_Legacy, shadingModel=MSM_Unlit, usedShadingModels=Unlit}
 nodes [
  n1 = ScalarParameter() {parameterName=SS}
  n2 = VectorParameter() {parameterName=AspectBaseNoise, defaultValue=<rgba 1 1 1 8>}
  n3 = MaterialFunctionCall!(UVs, XY Frames=Constant2Vector() {g=16, r=16}) {materialFunction=/Volumetrics/Content/VolumeTextures/MaterialFunctions/EncodeVolumeCoordinates}
  n4 = Add(
   A=Multiply(
    A=VectorParameter() {parameterName=Panner, defaultValue=<rgba 0 0 15 0>}.RGB
    B=MaterialFunctionCall(Speed=@n1) {materialFunction=&TimeWithSpeedVariable}.Frac Time
   )
   B=Multiply(A=Multiply(A=@n2.RGB, B=@n2.RGB), B=@n3.Result)
  )
  n5 = VectorParameter() {parameterName=AspectCurlNoise, defaultValue=<rgba 1 1 1 1>}
  n6 = Multiply(A=Multiply(A=@n5.RGB, B=@n5.RGB), B=@n3.Result)
  Add(A=@n4, B=Multiply(A=VectorNoise(World Position=@n6) {noiseFunction=VNF_CurlALU}, B=ScalarParameter() {parameterName=asdf}))
  Constant() {r=1}
  n7 = ScalarParameter() {parameterName=SS2}
  n8 = VectorParameter() {parameterName=AspectBaseNoise2, defaultValue=<rgba 1 1 1 8>}
  n9 = Add(
   A=Multiply(
    A=VectorParameter() {parameterName=Panner, defaultValue=<rgba 0 0 15 0>}.RGB
    B=MaterialFunctionCall(Speed=@n7) {materialFunction=&TimeWithSpeedVariable}.Frac Time
   )
   B=Multiply(A=Multiply(A=@n8.RGB, B=@n8.RGB), B=@n3.Result)
  )
  n10 = Add(
   A=Multiply(A=Noise!(World Position=@n9, FilterWidth) {noiseFunction=NOISEFUNCTION_GradientTex3D}, B=ScalarParameter() {parameterName=asdfdasf})
   B=Multiply(A=Noise(World Position=@n4) {noiseFunction=NOISEFUNCTION_GradientTex3D}, B=ScalarParameter() {parameterName=N})
  )
  n11 = ScalarParameter() {defaultValue=1, parameterName=Noise1}
  n12 = Subtract(A=@n3.Result, B=VectorParameter() {parameterName=P, defaultValue=<rgba 0.5 0.5 0.5 0>}.RGB)
  n13 = VectorParameter() {parameterName=Aspect, defaultValue=<rgba 1 1 1 0>}
  n14 = Divide(
   A=MaterialFunctionCall(Vector 3=Multiply(A=@n12, B=@n13.RGB)) {materialFunction=/Engine/Functions/Engine_MaterialFunctions02/Utility/VectorLength}.V3 Length
   B=ScalarParameter() {defaultValue=0.25, parameterName=R}
  )
  n15 = Multiply(A=Add(A=Multiply(A=@n10, B=@n11), B=OneMinus(None=@n14)), B=ScalarParameter() {defaultValue=1, parameterName=Temp})
  OneMinus(None=Power(Base=Saturate(None=@n15), Exp=ScalarParameter() {defaultValue=0.5, parameterName=TempPower}))
  n16 = MaterialFunctionCall(UVs=TextureCoordinate(), XY Frames=Constant() {r=16}) {materialFunction=&EncodeVolumeCoordinates}
  Saturate(None=Ceil(None=Subtract(A=@n16.Result) {constB=0.5}))
 ]
}
